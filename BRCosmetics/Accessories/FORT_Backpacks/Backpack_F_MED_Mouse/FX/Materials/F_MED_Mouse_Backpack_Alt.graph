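# Alternate surface material for the F_MED_Mouse backpack cosmetic, a Default Lit material flagged for skeletal mesh and mesh deformer use.
# The dot product of camera vector and vertex normal feeds two SmoothSteps that shade a blue pair and a red pair of constant colors.
# The lower-threshold mask also blends between the pairs, giving a view-angle color shift from blue at grazing angles to dark red facing the camera.
# The backpack's normal and mask textures are sampled alongside.
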Material /BRCosmetics/Accessories/FORT_Backpacks/Backpack_F_MED_Mouse/FX/Materials/F_MED_Mouse_Backpack_Alt
graph {
 properties {bUsedWithMeshDeformer=true, bUsedWithSkeletalMesh=true, pixelDepthOffsetMode=PDOM_Legacy, usedShadingModels=Default Lit}
 nodes [
  n1 = DotProduct(A=CameraVectorWS(), B=VertexNormalWS())
  n2 = Saturate(None=SmoothStep(Value=@n1) {constMax=0.5, constMin=0.1})
  n3 = SmoothStep(Value=@n1) {constMax=0.75, constMin=0.4}
  LinearInterpolate(
   A=LinearInterpolate(A=Constant3Vector() {constant=<rgba 0 0.3024 1 0>}, B=Constant3Vector() {constant=<rgba 0 0.0284 0.0938 0>}, Alpha=@n2)
   B=LinearInterpolate(A=Constant3Vector() {constant=<rgba 0.401 0.0118 0 0>}, B=Constant3Vector() {constant=<rgba 0.0208 0 0.0004 0>}, Alpha=Saturate(None=@n3))
   Alpha=@n2
  )
  TextureSample() {samplerType=SAMPLERTYPE_Normal, texture=/BRCosmetics/Accessories/FORT_Backpacks/Backpack_F_MED_Mouse/Textures/T_F_MED_Mouse_Backpack_N}
  TextureSample() {samplerType=SAMPLERTYPE_Masks, texture=/BRCosmetics/Accessories/FORT_Backpacks/Backpack_F_MED_Mouse/Textures/T_F_MED_Mouse_Backpack_S}
 ]
}
