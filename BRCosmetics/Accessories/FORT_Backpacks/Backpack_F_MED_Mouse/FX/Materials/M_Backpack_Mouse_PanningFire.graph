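# Translucent, unlit, two-sided panning fire material for the Mouse backpack backplate, usable on Niagara mesh particles.
# A static switch picks mesh UVs or polar UVs from a virtual plane to scroll banded noise, stored as a normal map for two-channel precision.
# The noise is wrapped with Frac, masked at its extremes and faded by the backplate's second UV set.
# The result indexes a curve atlas for color and feeds a smoothstep with its alpha.
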Material /BRCosmetics/Accessories/FORT_Backpacks/Backpack_F_MED_Mouse/FX/Materials/M_Backpack_Mouse_PanningFire
graph {
 properties {
  bUsedWithNiagaraMeshParticles=true
  blendMode=BLEND_Translucent
  pixelDepthOffsetMode=PDOM_Legacy
  previewMesh=/Game/Accessories/FORT_Backpacks/Backpack_F_MED_Mouse/FX/Meshes/SM_Backpack_Mouse_BackPlate
  shadingModel=MSM_Unlit
  twoSided=true
  usedShadingModels=Unlit
 }
 nodes [
  n1 = OneMinus(None=ComponentMask(None=TextureCoordinate() {coordinateIndex=1, desc=special UVs on the backplate}) {r=true})
  n2 = ScalarParameter() {defaultValue=0.5, parameterName=Backplate Fade Exponent}
  n3 = NamedRerouteUsage()
  n4 = TextureCoordinate()
  n5 = StaticBoolParameter() {defaultValue=true, parameterName=Use Polar UVs}
  n6 = MaterialFunctionCall(Float2=NamedRerouteDeclaration(None=StaticSwitch(True=@n3, False=@n4, Value=@n5)) {name=UVs, nodeColor=<rgba 0 1 0.7422 1>}) {materialFunction=&BreakOutFloat2Components}
  n7 = ScalarParameter() {defaultValue=6, parameterName=UV Scale U}
  n8 = Add(A=ScalarParameter() {parameterName=Offset}, B=Multiply(A=Time(), B=ScalarParameter() {defaultValue=0.666, parameterName=UV Speed V}))
  n9 = Add(A=@n6.G, B=Frac(None=@n8))
  n10 = TextureSample(UVs=AppendVector(A=Multiply(A=@n6.R, B=@n7), B=@n9)) {
   desc="Using a normal map to get higher quality,\r\nsince I just need 2 channels"
   samplerType=SAMPLERTYPE_Normal
   texture=/BRCosmetics/Accessories/FORT_Backpacks/Backpack_F_MED_Mouse/FX/Textures/T_Backpack_Mouse_BandedNoise
  }
  n11 = Add(A=ConstantBiasScale(None=ComponentMask(None=@n10.RGB) {g=true, r=true}), B=Constant2Vector() {g=0.5})
  n12 = MaterialFunctionCall(Float2=Frac(None=Subtract(A=@n11, B=@n8))) {materialFunction=&BreakOutFloat2Components}
  n13 = ConstantBiasScale(None=Frac(None=@n9)) {bias=-0.5, scale=2}
  n14 = Divide(A=Subtract(A=Abs(None=@n13)) {constB=0.5}, B=ScalarParameter() {defaultValue=0.0001, parameterName=Blend Softness})
  n15 = LinearInterpolate(A=@n12.R, B=@n12.R, Alpha=Saturate(None=Add(A=@n14) {constB=0.5}))
  n16 = Frac(None=@n15)
  n17 = Round(None=ComponentMask(None=NamedRerouteUsage()) {g=true})
  n18 = Abs(None=ConstantBiasScale(None=@n16) {bias=-0.5, scale=2})
  n19 = ScalarParameter() {defaultValue=0.7, parameterName=Mask Offset}
  n20 = Saturate(None=Divide(A=Subtract(A=@n18, B=@n19), B=ScalarParameter() {defaultValue=0.166, parameterName=Mask Softness}))
  n21 = Multiply(
   A=LinearInterpolate(B=Power(Base=@n1, Exp=@n2) {constExponent=3}, Alpha=ScalarParameter() {defaultValue=1, parameterName=Use Backplate Fade}) {constA=1}
   B=Saturate(None=LinearInterpolate(A=@n16, B=@n17, Alpha=@n20) {desc="Masking the top and\r\nbottom weirdness"})
  )
  n22 = CurveAtlasRowParameter(CurveTime=Saturate(None=@n21)) {
   atlas=/BRCosmetics/Accessories/FORT_Backpacks/Backpack_F_MED_Mouse/FX/Textures/CA_Mouse_Backpack
   curve=/BRCosmetics/Accessories/FORT_Backpacks/Backpack_F_MED_Mouse/FX/Textures/C_MouseBackpack_Back
   defaultValue=1
  }
  MaterialFunctionCall(Emissive=@n22) {materialFunction=&MF_FX_EmissiveAdjustments}
  n23 = VectorParameter() {
   desc=You'll likely need to transform this stuff once you stick it inside a niagara
   parameterName="Pivot Offset && UV Scale"
   defaultValue=<rgba 0 0 6 41>
  }
  n24 = MaterialFunctionCall(Plane Normal Axis=VertexNormalWS(), Plane Center=Subtract(A=ObjectPositionWS(), B=@n23.RGB), UV Scale=@n23.A) {materialFunction=&VirtualPlaneCoordinates}
  NamedRerouteDeclaration(None=OneMinus(None=MaterialFunctionCall!(UV=@n24.UVs, Twist Center, Twist Edges) {materialFunction=/Game/Accessories/FORT_Backpacks/Backpack_F_MED_Celestial_Glow/FX/Materials/MF_PolarUV}.Polar UVs)) {name=Polar UVs, nodeColor=<rgba 1 0.5391 0 1>}
  ParticlePositionWS()
  Power(Base=@n15, Exp=ScalarParameter() {defaultValue=1, parameterName=Param_1})
  SmoothStep(Value=Multiply(A=@n22.A, B=@n21)) {constMax=0.25, constMin=0.15}
 ]
}
