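# Unlit, two-sided translucent Niagara sprite material that draws an energy spike effect for the Mouse backpack cosmetic.
# A tapered UV-math shape is modulated by smoke noise, panned by a dynamic parameter's Time.
# The StateLerp scalar blends between two states, changing noise weighting, edge thresholds and brightness.
# The resulting mask picks color from a curve atlas for emissive and, with particle alpha and depth fade, sets opacity.
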Material /BRCosmetics/Accessories/FORT_Backpacks/Backpack_F_MED_Mouse/FX/Materials/M_MouseBackpack_EnergySpike
graph {
 properties {
  bUsedWithNiagaraSprites=true
  blendMode=BLEND_Translucent
  pixelDepthOffsetMode=PDOM_Legacy
  shadingModel=MSM_Unlit
  twoSided=true
  usedShadingModels=Unlit
 }
 nodes [
  n1 = TextureCoordinate()
  n2 = ComponentMask(None=@n1) {g=true}
  n3 = Add(A=@n1) {constB=-0.5}
  n4 = Power(Base=ComponentMask(None=Abs(None=@n3)) {r=true}) {constExponent=1}
  n5 = Multiply(A=Subtract(A=@n2, B=Multiply(A=@n4) {constB=2.5}), B=SmoothStep(Value=OneMinus(None=@n2)) {constMax=0.2})
  n6 = DynamicParameter() {paramNames [RandomTime,Time,Param3,Param4]}
  n7 = TextureSample(UVs=Panner(Coordinate=TextureCoordinate() {vTiling=0.5}, Time=Add(A=@n6.Time, B=@n6.Time)) {speedY=0.5}) {texture=&T_Inky_Smoke_Tile_Chromatic}
  n8 = ConstantBiasScale(None=@n7.R)
  n9 = Add(A=@n7.R) {constB=0.25}
  n10 = Saturate(None=LinearInterpolate(A=Multiply(A=@n5, B=@n8), B=Multiply(A=@n5, B=@n9), Alpha=NamedRerouteUsage()))
  n11 = LinearInterpolate(A=SmoothStep(Value=@n10) {constMax=0.35, constMin=0.1}, B=SmoothStep(Value=@n10) {constMax=0.25, constMin=0.05}, Alpha=NamedRerouteUsage())
  n12 = CurveAtlasRowParameter(CurveTime=@n11) {
   atlas=/BRCosmetics/Accessories/FORT_Backpacks/Backpack_F_MED_Mouse/FX/Textures/CA_Mouse_Backpack
   curve=/BRCosmetics/Accessories/FORT_Backpacks/Backpack_F_MED_Mouse/FX/Textures/C_MouseBackpack_Spike
   defaultValue=2
  }
  MaterialFunctionCall(Emissive=LinearInterpolate(A=@n12, B=Multiply(A=@n12) {constB=1.75}, Alpha=NamedRerouteUsage())) {materialFunction=&MF_FX_EmissiveAdjustments}
  Multiply(
   A=Multiply(A=@n11, B=ParticleColor().A)
   B=DepthFade(FadeDistance=ScalarParameter() {defaultValue=15, parameterName=Param_1}) {fadeDistanceDefault=25}
  )
  NamedRerouteDeclaration(None=ScalarParameter() {parameterName=StateLerp}) {name=StateLerp, nodeColor=<rgba 0 0.5078 1 1>}
 ]
}
