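# Masked Default Lit Niagara material for sprite and mesh particles, drawing a solid spike for the Mouse backpack effects.
# The spike shape is procedural, without textures: V minus scaled distance from center U feeds smoothsteps for a two-color gradient and a hard-edged mask.
# A squared UV-derived vector and an inverted horizontal sphere mask add further shading terms.
# Scalar parameters expose metallic, roughness and specular.
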Material /BRCosmetics/Accessories/FORT_Backpacks/Backpack_F_MED_Mouse/FX/Materials/M_MouseBackpack_Solidspike
graph {
 properties {
  bUsedWithNiagaraMeshParticles=true
  bUsedWithNiagaraSprites=true
  blendMode=BLEND_Masked
  pixelDepthOffsetMode=PDOM_Legacy
  usedShadingModels=Default Lit
 }
 nodes [
  n1 = TextureCoordinate()
  n2 = Abs(None=Add(A=@n1) {constB=-0.5})
  n3 = Subtract(A=ComponentMask(None=@n1) {g=true}, B=Multiply(A=ComponentMask(None=@n2) {r=true}) {constB=2.5})
  LinearInterpolate(
   A=VectorParameter() {parameterName=Color2, defaultValue=<rgba 0.5521 0.0444 0.0564 1>}.RGB
   B=VectorParameter() {parameterName=Color1, defaultValue=<rgba 0.026 0.0021 0.0027 1>}.RGB
   Alpha=Saturate(None=SmoothStep(Value=@n3) {constMax=0.25})
  )
  n4 = TextureCoordinate()
  n5 = Abs(None=Add(A=@n4) {constB=-0.5})
  n6 = MaterialFunctionCall(R=ComponentMask(None=@n4) {r=true}, G=Constant() {r=0.4}, B=OneMinus(None=ComponentMask(None=@n5) {r=true})) {materialFunction=&AppendMany}
  Multiply(A=@n6.RGB, B=@n6.RGB)
  OneMinus(None=SphereMask(
    A=TextureCoordinate() {vTiling=0}
    B=Constant2Vector() {r=0.5}
    Radius=ScalarParameter() {defaultValue=0.5, parameterName=Param_1}
    Hardness=ScalarParameter()
   ))
  ScalarParameter() {parameterName=Metallic}
  ScalarParameter() {parameterName=Rough}
  ScalarParameter() {parameterName=Specular}
  SmoothStep(Value=@n3) {constMax=0}
 ]
}
