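# Niagara idle effect for the Mouse backpack cosmetic, with a heart, pulsing core flare and light, lightning and spike sprites.
# All emitters follow the User Skeletal Mesh through bone sampling with local offsets.
# User.StateLerp intensifies the effect, enlarging the flare and light, swapping small lightning for large and adding main spikes.
# Spikes form arcs via cylinder shape locations, while energy sprites align from eight offsets and vary width with camera angle.
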NiagaraSystem /BRCosmetics/Accessories/FORT_Backpacks/Backpack_F_MED_Mouse/FX/Systems/NS_Backpack_Mouse_Idle
dependencies {
 usedDataInterfaces [
  &NiagaraDataInterfaceCurve
  &NiagaraDataInterfaceArrayFloat3
  &NiagaraDataInterfaceArrayInt32
  &NiagaraDataInterfaceVector2DCurve
  &NiagaraDataInterfaceArrayFloat
  &NiagaraDataInterfaceCamera
 ]
 usedDynamicInputs []
 usedModules [&SystemState]
}
emitters {
 CoreFlare {
  emitterUpdateScript [
   &EmitterState(Life Cycle Mode=::NewEnumerator1, Loop Behavior=::NewEnumerator1, Loop Duration Mode=::NewEnumerator1, UseLoopDelay=false)
   &SpawnBurst_Instantaneous(Use Spawn Probability=false, UseLoopCountLimit=false)
  ]
  particleSpawnScript [
   /SetVariables(Particles.LensColor:LinearColor=<rgba 0.1227 0.2 0.1819 1>) {isSetParameters=true}
   &InitializeParticle(
    Color=<rgba 1 0.5331 0.9326 0.25>
    Lifetime=4.5
    Sprite Size=<15 20>
    Sprite Size Mode=::NewEnumerator4
    UsePositionOffset=false
    Write Material Random=true
   )
   &CameraOffset(Camera Offset Amount=7)
   /SetVariables(Particles.VisibilityTag=&SelectIntFromArray(Direct Array Index=&SpawnGroup(), Int Selection Array=DataInterface() {propertyValues {IntData [0,1]}})) {isSetParameters=true}
  ]
  particleUpdateScript [
   &ParticleState(Kill Particles When Lifetime Has Elapsed=false, Let Infinitely Lived Particles Die When Emitter Deactivates=true)
   &ScaleSpriteSize(Scale Sprite Size Mode=::NewEnumerator0, Uniform Scale Factor=&Lerp_Float(A=1, B=1.75, Alpha=$User.StateLerp))
   &SkeletalMeshLocation(Bone Sampling Mode=::NewEnumerator4, MaskSpawnGroup=false, Sampled Position Offset=<0 -4.5 -1.75>, Skeletal Mesh=$User.Skeletal Mesh)
   /SetVariables(Particles.DarkRadialColor:LinearColor=<rgba 0.0521 0 0.0304 1>) {isSetParameters=true}
   /SetVariables(Particles.DarkRadialSize:Vector2f=&Multiply_Vector2D(A=$Particles.SpriteSize, B=<0.45 0.8>)) {isSetParameters=true}
   &ScaleColor(
    Scale Alpha=&Sine(
     Normalized Angle=$Emitter.Age
     Period=&Waveform("Waveform [1]"=::NewEnumerator0, "[1] Frequency"=1, "[1] Amplitude Min/Max"=<0 1>)
     Scale=0.25
     Bias=1
    )
    Scale RGB=&Multiply_VectorByFloat(Vector=<1 1 1>, Float=&Lerp_Float(A=1, B=2, Alpha=$User.StateLerp))
    ScaleA=true
    ScaleRGB=true
   )
   &FN_Light_Attributes(
    Color=$Particles.Color
    Create Color Attribute=true
    Create Exponent Attribute=true
    Create Position Attribute=true
    Create Radius Attribute=true
    Exponent=4
    Position=<0 -20 2>
    Radius=&Lerp_Float(A=50, B=75, Alpha=$User.StateLerp)
    Scale Color=&Waveform("Waveform [1]"=::NewEnumerator12, "[1] Phase Scale"=10, "[1] Frequency"=1, "[1] Amplitude Min/Max"=<0.2 1>, "[1] Bias"=0.5)
   )
  ]
  rendererClasses [&NiagaraSpriteRendererProperties, &NiagaraLightRendererProperties]
  renderers [
   {rendererClass=&NiagaraSpriteRendererProperties}
   {rendererClass=&NiagaraSpriteRendererProperties, rendererIndex=1}
   {rendererClass=&NiagaraLightRendererProperties, rendererIndex=2}
  ]
 }
 Heart {
  emitterUpdateScript [
   &EmitterState(Life Cycle Mode=::NewEnumerator1, Loop Behavior=::NewEnumerator1, Loop Duration Mode=::NewEnumerator1, UseLoopDelay=false)
   &SpawnBurst_Instantaneous(Spawn Count=2, Use Spawn Probability=false, UseLoopCountLimit=false)
   /SetVariables(Emitter.StateLerp:NiagaraFloat=$User.StateLerp) {isSetParameters=true, module=SetVariables_#1}
  ]
  particleSpawnScript [&InitializeParticle(
    Color Mode=::NewEnumerator0
    Mesh Scale=&SelectVectorFromArray(
     Direct Array Index=&ReturnExecIndex()
     Vector Selection Array=DataInterface() {propertyValues {FloatData [<1 1 1>, <1 0.95 1>]}}
    )
    Mesh Scale Mode=::NewEnumerator4
    UsePositionOffset=false
    Write Material Random=true
   )]
  particleUpdateScript [
   &ParticleState(Kill Particles When Lifetime Has Elapsed=false)
   &SkeletalMeshLocation(
    Bone Sampling Mode=::NewEnumerator4
    MaskSpawnGroup=false
    Sampled Position Offset=&SelectVectorFromArray(
     Direct Array Index=&ReturnExecIndex()
     Vector Selection Array=DataInterface() {propertyValues {FloatData [<0 1 -0.35>, <0 0.75 0>]}}
    )
    Skeletal Mesh=$User.Skeletal Mesh
   )
   /SetVariables(Particles.SpriteFacing=<0 -1 0>) {isSetParameters=true}
   /SetVariables(Particles.VisibilityTag=&SelectIntFromArray(Direct Array Index=&ReturnExecIndex(), Int Selection Array=DataInterface() {propertyValues {IntData [0,1]}})) {isSetParameters=true}
  ]
  rendererClasses [&NiagaraMeshRendererProperties]
  renderers [
   {rendererClass=&NiagaraMeshRendererProperties, rendererIndex=0}
   {rendererClass=&NiagaraMeshRendererProperties, rendererIndex=1}
  ]
 }
 Lightning_Large {
  emitterUpdateScript [
   &EmitterState(
    Life Cycle Mode=::NewEnumerator1
    Loop Duration=&RandomRangeFloat(Minimum=0.1, Maximum=0.3)
    Recalculate Duration Each Loop=true
    UseLoopDelay=false
   )
   &SpawnRate(Spawn Probability=&Lerp_Float(A=0, B=0.33, Alpha=$User.StateLerp), SpawnRate=55, Use Spawn Probability=true)
  ]
  particleSpawnScript [
   &InitializeParticle(
    Color=<rgba 0.1373 0.5961 1 1>
    Lifetime Max=0.3
    Lifetime Min=0.2
    Lifetime Mode=::NewEnumerator1
    Mesh Scale Max=<0.8 0.8 2.25>
    Mesh Scale Min=<0.5 0.5 1>
    Mesh Scale Mode=::NewEnumerator2
    Position Offset=<0 -4 -3>
    Sprite Size Mode=::NewEnumerator1
    Uniform Sprite Size Max=25
    Uniform Sprite Size Min=20
    UsePositionOffset=true
    Write Material Random=true
   )
   &InitialMeshOrientation(
    Mesh Orientation Mode=::NewEnumerator2
    Rotation=&RandomRangeVector(Minimum=<-0.15 0 -0.3>, Maximum=<0.15 0 -0.7>)
    Rotation Coordinate Space=::NewEnumerator2
    Use Rotation Vector=true
   )
   /SetVariables(Particles.SpriteCameraOffset=3) {isSetParameters=true, module=SetVariables_#2}
   &CameraOffset()
   /SetVariables(Particles.MeshIndex=&RandomRangeInt(Minimum=0, Maximum=2)) {isSetParameters=true}
  ]
  particleUpdateScript [
   &ParticleState()
   &ScaleSpriteSize(Uniform Curve Sprite Scale=DataInterface() {propertyValues {Curve {keys [
        {arriveTangent=9.0725, interpMode=RCIM_Cubic, leaveTangent=9.0725, tangentMode=RCTM_User, time=0, value=0.5}
        {arriveTangent=-2.5249, interpMode=RCIM_Cubic, leaveTangent=-2.5249, tangentMode=RCTM_User, time=0.1408, value=0.9847}
        {arriveTangent=-0.0405, interpMode=RCIM_Cubic, leaveTangent=-0.0405, tangentMode=RCTM_User, time=0.4, value=0}
       ]}}})
   &SkeletalMeshLocation(
    Bone Sampling Mode=::NewEnumerator4
    MaskSpawnGroup=false
    Sampled Position Offset=&SelectVectorFromArray(
     Array Sampling Mode=::NewEnumerator0
     Direct Array Index=$Emitter.LoopCount
     Vector Selection Array=DataInterface() {propertyValues {FloatData [<3 -5 -2>, <0 -5 -3>, <-3 -5 -2>]}}
     Random Seed=0
    )
    Skeletal Mesh=$User.Skeletal Mesh
   )
   &ScaleColor(
    Scale Alpha=&FloatFromCurve(
     FloatCurve=DataInterface() {propertyValues {Curve {keys [
         {arriveTangent=9.0725, interpMode=RCIM_Cubic, leaveTangent=9.0725, tangentMode=RCTM_User, time=0, value=0}
         {arriveTangent=-2.5249, interpMode=RCIM_Cubic, leaveTangent=-2.5249, tangentMode=RCTM_User, time=0.1408, value=0.9847}
         {arriveTangent=-0.0405, interpMode=RCIM_Cubic, leaveTangent=-0.0405, tangentMode=RCTM_User, time=0.5, value=0}
        ]}}}
     Scale Curve=0.25
    )
    ScaleA=true
    ScaleRGB=true
   )
  ]
  rendererClasses [&NiagaraMeshRendererProperties, &NiagaraSpriteRendererProperties]
  renderers [
   {rendererClass=&NiagaraMeshRendererProperties, rendererIndex=0}
   {rendererClass=&NiagaraMeshRendererProperties, rendererIndex=1}
   {rendererClass=&NiagaraSpriteRendererProperties, rendererIndex=2}
  ]
 }
 Lightning_Small {
  emitterUpdateScript [
   &EmitterState(
    Life Cycle Mode=::NewEnumerator1
    Loop Duration=&RandomRangeFloat(Minimum=0.1, Maximum=0.3)
    Recalculate Duration Each Loop=true
    UseLoopDelay=false
   )
   &SpawnRate(Spawn Probability=&Lerp_Float(A=0.22, B=0, Alpha=$User.StateLerp), SpawnRate=5, Use Spawn Probability=true)
  ]
  particleSpawnScript [
   &InitializeParticle(
    Color=<rgba 0.1373 0.5961 1 1>
    Lifetime Max=0.3
    Lifetime Min=0.2
    Lifetime Mode=::NewEnumerator1
    Mesh Scale Max=<0.8 0.8 1>
    Mesh Scale Min=<0.25 0.25 0.5>
    Mesh Scale Mode=::NewEnumerator2
    Position Offset=<0 -4 -3>
    Sprite Size Mode=::NewEnumerator3
    Uniform Sprite Size=&RandomRangeFloat(Minimum=15, Maximum=20)
    UsePositionOffset=true
    Write Material Random=true
   )
   &InitialMeshOrientation(
    Mesh Orientation Mode=::NewEnumerator2
    Orientation Vector=<-1 0 0>
    Rotation=&RandomRangeVector(Minimum=<0 0 -0.3>, Maximum=<0 0 -0.7>)
    Rotation Coordinate Space=::NewEnumerator2
    Use Rotation Vector=true
   )
   /SetVariables(Particles.SpriteCameraOffset=3) {isSetParameters=true, module=SetVariables_#2}
   &CameraOffset()
  ]
  particleUpdateScript [
   &ParticleState()
   &ScaleSpriteSize(Uniform Curve Sprite Scale=DataInterface() {propertyValues {Curve {keys [
        {arriveTangent=9.0725, interpMode=RCIM_Cubic, leaveTangent=9.0725, tangentMode=RCTM_User, time=0, value=0.5}
        {arriveTangent=-2.5249, interpMode=RCIM_Cubic, leaveTangent=-2.5249, tangentMode=RCTM_User, time=0.1408, value=0.9847}
        {arriveTangent=-0.0405, interpMode=RCIM_Cubic, leaveTangent=-0.0405, tangentMode=RCTM_User, time=0.5, value=0}
       ]}}})
   &SkeletalMeshLocation(
    Bone Sampling Mode=::NewEnumerator4
    MaskSpawnGroup=false
    Sampled Position Offset=&SelectVectorFromArray(
     Array Sampling Mode=::NewEnumerator0
     Direct Array Index=$Emitter.LoopCount
     Vector Selection Array=DataInterface() {propertyValues {FloatData [<2 -4.8 -3>, <0 -4.8 -4>, <-2 -4.8 -3>]}}
     Random Seed=0
    )
    Skeletal Mesh=$User.Skeletal Mesh
   )
   &ScaleColor(
    Scale Alpha=&FloatFromCurve(
     FloatCurve=DataInterface() {propertyValues {Curve {keys [
         {arriveTangent=9.0725, interpMode=RCIM_Cubic, leaveTangent=9.0725, tangentMode=RCTM_User, time=0, value=0}
         {arriveTangent=-2.5249, interpMode=RCIM_Cubic, leaveTangent=-2.5249, tangentMode=RCTM_User, time=0.1408, value=0.9847}
         {arriveTangent=-0.0405, interpMode=RCIM_Cubic, leaveTangent=-0.0405, tangentMode=RCTM_User, time=0.5, value=0}
        ]}}}
     Scale Curve=0.5
    )
    ScaleA=true
    ScaleRGB=true
   ) {module=ScaleColor001}
  ]
  rendererClasses [&NiagaraMeshRendererProperties, &NiagaraSpriteRendererProperties]
  renderers [
   {rendererClass=&NiagaraMeshRendererProperties, rendererIndex=0}
   {rendererClass=&NiagaraMeshRendererProperties, rendererIndex=1}
   {rendererClass=&NiagaraSpriteRendererProperties, rendererIndex=2}
  ]
 }
 Spikes {
  emitterUpdateScript [
   &EmitterState(
    Life Cycle Mode=::NewEnumerator1
    Loop Duration=&Lerp_Float(
     A=&RandomRangeFloat(Minimum=0.2, Maximum=0.4)
     B=&RandomRangeFloat(Minimum=0.15, Maximum=0.3)
     Alpha=$User.StateLerp
    )
    UseLoopDelay=false
   )
   &SpawnRate(Spawn Probability=0.33, SpawnRate=&Lerp_Float(A=22, B=33, Alpha=$User.StateLerp), Use Spawn Probability=true)
   /SetVariables(Emitter.StateLerp:NiagaraFloat=$User.StateLerp) {isSetParameters=true, module=SetVariables_#1}
  ]
  particleSpawnScript [
   &InitializeParticle(
    Color Mode=::NewEnumerator0
    Lifetime=&Lerp_Float(
     A=&RandomRangeFloat(Minimum=0.3, Maximum=0.45)
     B=&RandomRangeFloat(Minimum=0.2, Maximum=0.4)
     Alpha=$User.StateLerp
    )
    Sprite Size=&Multiply_Vector2D_ByFloat(
     Float=&RandomRangeFloat(Minimum=0.7, Maximum=1)
     Vector2D=&MakeVector2D(
      X=&Lerp_Float(A=2, B=3, Alpha=$User.StateLerp)
      Y=&Lerp_Float(A=7, B=9, Alpha=$User.StateLerp)
     )
    )
    Sprite Size Mode=::NewEnumerator4
    UsePositionOffset=false
    Write Material Random=true
   )
   &CameraOffset(Camera Offset Amount=2)
  ]
  particleUpdateScript [
   &SkeletalMeshLocation(
    Bone Sampling Mode=::NewEnumerator4
    MaskSpawnGroup=false
    Sampled Scale Coordinate Space:ENiagaraCoordinateSpace=::World
    Skeletal Mesh=$User.Skeletal Mesh
   )
   &ShapeLocation(
    Cylinder Distribution=::NewEnumerator2
    Cylinder Height=0
    Cylinder Radius=4.5
    MaskSpawnGroup=false
    Offset=<0 -4.35 -7>
    Offset Mode=::NewEnumerator0
    Rotation Axis=<0 1 0>
    Rotation Coordinate Space=::Simulation
    Rotation Mode=::NewEnumerator1
    Shape Origin=$Particles.SkeletalMeshLocation.SampledPosition
    Shape Primitive=::NewEnumerator1
    U Position=&RandomRangeFloat(Minimum=-0.17, Maximum=-0.33)
    UseOwnerScale=true
   )
   /SetVariables(Particles.SpriteAlignment=&AddVectorToPosition(Position=Linked() {linkedVariable {name=Particles.ShapeLocation.ShapeNormal, type=Vector3f}}, Vector=<0 -0.45 0>)) {isSetParameters=true, module=SetVariables_#3}
   &ParticleState()
   &ScaleSpriteSize(
    Non-Uniform Curve Sprite Scale=DataInterface() {propertyValues {
      XCurve {keys [
        {arriveTangent=4.1785, interpMode=RCIM_Cubic, leaveTangent=4.1785, tangentMode=RCTM_User, time=0, value=0}
        {arriveTangent=0.9995, interpMode=RCIM_Cubic, leaveTangent=0.9995, tangentMode=RCTM_User, time=0.2223, value=0.6825}
        {arriveTangent=-0.0405, interpMode=RCIM_Cubic, leaveTangent=-0.0405, tangentMode=RCTM_User, time=1, value=1}
       ]}
      YCurve {keys [
        {arriveTangent=2.6795, interpMode=RCIM_Cubic, leaveTangent=2.6795, tangentMode=RCTM_User, time=0, value=0}
        {arriveTangent=-0.0405, interpMode=RCIM_Cubic, leaveTangent=-0.0405, tangentMode=RCTM_User, time=0.5, value=1}
       ]}
     }}
    Scale Sprite Size Mode=::NewEnumerator3
   )
   &ScaleColor(
    Scale Alpha=&FloatFromCurve(FloatCurve=DataInterface() {propertyValues {Curve {keys [
         {interpMode=RCIM_Cubic, tangentMode=RCTM_User, time=0, value=0}
         <at 0.25 1>
         {interpMode=RCIM_Cubic, tangentMode=RCTM_User, time=0.75, value=1}
         {tangentMode=RCTM_User, time=1, value=0}
        ]}}})
    ScaleA=true
    ScaleRGB=true
   )
  ]
  rendererClasses [&NiagaraSpriteRendererProperties]
  renderers [
   {rendererClass=&NiagaraSpriteRendererProperties}
   {rendererClass=&NiagaraSpriteRendererProperties, rendererIndex=1}
  ]
 }
 SpikesEnergy_Main {
  emitterUpdateScript [
   &EmitterState(Life Cycle Mode=::NewEnumerator1, Loop Duration=1, UseLoopDelay=false)
   &SpawnRate(SpawnRate=32, Use Spawn Probability=false)
   /SetVariables(Emitter.StateSwitch:NiagaraFloat=$User.StateLerp) {isSetParameters=true, module=SetVariables_#4}
  ]
  particleSpawnScript [
   &DynamicMaterialParameters(
    Index 0 Param 1=&RandomRangeFloat(Minimum=0, Maximum=1)
    Module.Param0WriteEnabled=true
    Module.Param1WriteEnabled=true
    Module.Param2WriteEnabled=true
    Module.Param3WriteEnabled=true
   )
   &FN_SetSpawnGroup(Number of Spawn Groups=8)
   &InitializeParticle(
    Lifetime=0.5
    Sprite Size=&Multiply_Vector2D_ByFloat(
     Float=&Multiply_Float(
      A=1
      B=&SelectFloatFromArray(
       Direct Array Index=$Particles.SpawnGroup
       Float Selection Array=DataInterface() {propertyValues {FloatData [0.9,1.1,1.2,0.9,0.8,0.9,0.8,0.7]}}
      )
     )
     Vector2D=&MakeVector2D(X=10, Y=&Lerp_Float(A=8, B=17, Alpha=$User.StateLerp))
    )
    Sprite Size Mode=::NewEnumerator4
    UsePositionOffset=false
    Write Material Random=true
   )
   &CameraOffset(Camera Offset Amount=-1)
  ]
  particleUpdateScript [
   &SkeletalMeshLocation(
    Bone Sampling Mode=::NewEnumerator4
    MaskSpawnGroup=false
    Sampled Position Offset=&SelectVectorFromArray(
     Direct Array Index=$Particles.SpawnGroup
     Vector Selection Array=DataInterface() {propertyValues {FloatData [<2 -3.4 2.5>, <-2 -3.4 2.5>, <4.5 -3.1 2.8>, <-5 -3.25 3>, <6.76 -2.7 0.5>, <-7.25 -2.7 1.25>, <5 -2.2 -4>, <-5 -2.2 -4>]}}
    )
    Sampled Scale Coordinate Space:ENiagaraCoordinateSpace=::World
    Skeletal Mesh=$User.Skeletal Mesh
   )
   /SetVariables(Particles.LookatPos:Vector3f=&SubtractVectorFromPosition(
     Position=$Particles.SkeletalMeshLocation.SampledPosition
     Vector=&ConvertPositionToVector(
      Input Position=&Lerp_Vector(A=<0 3 4>, B=<0 3 5>, Alpha=$User.StateLerp)
      Conversion Operation:ENiagara_LWCConvertPosToVec=::NewEnumerator0
     )
    )) {isSetParameters=true}
   /SetVariables(Particles.SizeFade:NiagaraFloat=&NormalizeFloat(
     Value=&Power(
      a=&Abs_Float(Float:NiagaraFloat=&DotProduct(
        Vector A=&ReturnCameraVector(Camera Vectors=::NewEnumerator1, Coordinate Space=::Simulation)
        Vector B=&TransformVector(Vector=<0 1 0>, Destination Space=::Simulation)
       ))
      b=2
     )
     Min=0
     Max=1
    )) {isSetParameters=true}
   &ParticleState()
   &ScaleSpriteSize(
    Non-Uniform Curve Scale=&MakeVector2D(X=&Lerp_Float(A=0.75, B=1, Alpha=$Particles.SizeFade), Y=1)
    Non-Uniform Curve Sprite Scale=DataInterface() {propertyValues {
      XCurve {keys [
        {arriveTangent=3.7858, interpMode=RCIM_Cubic, leaveTangent=3.7858, leaveTangentWeight=0.3333, tangentMode=RCTM_User, time=0, value=0}
        {arriveTangentWeight=0.0833, interpMode=RCIM_Cubic, leaveTangentWeight=0.25, tangentMode=RCTM_User, time=0.25, value=1}
        {arriveTangent=-3.7521, arriveTangentWeight=0.25, interpMode=RCIM_Cubic, leaveTangent=-3.7521, tangentMode=RCTM_User, time=1, value=0}
       ]}
      YCurve {keys [
        {arriveTangent=3.7858, interpMode=RCIM_Cubic, leaveTangent=3.7858, leaveTangentWeight=0.3263, tangentMode=RCTM_User, time=0, value=0}
        {arriveTangentWeight=0.0833, interpMode=RCIM_Cubic, leaveTangentWeight=0.25, tangentMode=RCTM_User, time=0.25, value=1}
        {arriveTangent=-3.7521, arriveTangentWeight=0.9708, interpMode=RCIM_Cubic, leaveTangent=-3.7521, tangentMode=RCTM_User, time=1, value=0}
       ]}
     }}
    Scale Sprite Size Mode=::NewEnumerator3
   )
   /SetVariables(Particles.GlowSize=&Multiply_Vector2D(A=$Particles.SpriteSize, B=&Multiply_Vector2D_ByFloat(Float=1, Vector2D=<1.25 1.5>))) {isSetParameters=true, module=SetVariables_#5}
   /SetVariables(Particles.SpriteAlignment=&NormalizeVector(
     Vector To Normalize=&Subtract_Vector(A=Linked() {linkedVariable {name=Particles.Position, type=NiagaraPosition}}, B=$Particles.LookatPos)
     Safe Normalize=true
     Safe Fallback Threshold=0
     Safe Fallback Vector=<0 0 0.001>
    )) {isSetParameters=true}
   &ScaleColor(
    Scale Alpha=&FloatFromCurve(FloatCurve=DataInterface() {propertyValues {Curve {keys [
         {interpMode=RCIM_Cubic, tangentMode=RCTM_User, time=0, value=0}
         {interpMode=RCIM_Cubic, tangentMode=RCTM_User, time=0.5, value=1}
         {tangentMode=RCTM_User, time=1, value=0}
        ]}}})
    Scale RGB=&MakeVectorFromLinearColorRGB(Color=<rgba 1 1 1 1>)
    ScaleA=true
    ScaleRGB=true
   )
   &DynamicMaterialParameters(
    Index 0 Param 2=&FloatFromCurve(
     FloatCurve=DataInterface() {propertyValues {Curve {keys [{tangentMode=RCTM_User, time=0, value=0}, {tangentMode=RCTM_User, time=1, value=1}]}}}
     Scale Curve=&Lerp_Float(A=0.5, B=2, Alpha=$User.StateLerp)
    )
    Module.Param0WriteEnabled=false
    Module.Param1WriteEnabled=true
    Module.Param2WriteEnabled=true
    Module.Param3WriteEnabled=true
   ) {module=DynamicMaterialParameters001}
  ]
  rendererClasses [&NiagaraSpriteRendererProperties]
  renderers [{rendererClass=&NiagaraSpriteRendererProperties}]
 }
 SpikesMain {
  emitterUpdateScript [
   &EmitterState()
   &SpawnRate(SpawnRate=&Lerp_Float(A=0, B=10, Alpha=$User.StateLerp), Use Spawn Probability=false)
   /SetVariables(Emitter.StateSwitch:NiagaraFloat=$User.StateLerp) {isSetParameters=true, module=SetVariables_#4}
  ]
  particleSpawnScript [
   &InitializeParticle(
    Color Mode=::NewEnumerator0
    Lifetime=&Lerp_Float(
     A=&RandomRangeFloat(Minimum=0.5, Maximum=1)
     B=&RandomRangeFloat(Minimum=0.35, Maximum=0.7)
     Alpha=$User.StateLerp
    )
    Sprite Size=&Multiply_Vector2D_ByFloat(
     Float=&Multiply_Float(
      A=&Lerp_Float(A=0.8, B=1.1, Alpha=$User.StateLerp)
      B=&RandomRangeFloat(Minimum=0.7, Maximum=1)
     )
     Vector2D=<3.5 8>
    )
    Sprite Size Mode=::NewEnumerator4
    UsePositionOffset=false
    Write Material Random=true
   )
   &FN_SetSpawnGroup(Number of Spawn Groups=2)
  ]
  particleUpdateScript [
   &SkeletalMeshLocation(
    Bone Sampling Mode=::NewEnumerator4
    MaskSpawnGroup=false
    Sampled Scale Coordinate Space:ENiagaraCoordinateSpace=::World
    Skeletal Mesh=$User.Skeletal Mesh
   )
   &ShapeLocation(
    Cylinder Distribution=::NewEnumerator2
    Cylinder Height=0
    Cylinder Radius=9
    MaskSpawnGroup=false
    Offset=&RandomRangeVector(Minimum=<0 -2.5 -6>, Maximum=<0 -3.5 -6>)
    Offset Mode=::NewEnumerator0
    Rotation Axis=<0 1 0>
    Rotation Coordinate Space=::Simulation
    Rotation Mode=::NewEnumerator1
    Shape Origin=$Particles.SkeletalMeshLocation.SampledPosition
    Shape Primitive=::NewEnumerator1
    U Position=&MaskFloatBySpawnGroup(
     Masked Float=&RandomRangeFloat(Minimum=-0.11, Maximum=-0.22)
     Passthrough Float=&RandomRangeFloat(Minimum=-0.28, Maximum=-0.39)
     Spawn Group Mask=1
    )
    UseOwnerScale=true
   )
   /SetVariables(Particles.SpriteAlignment=&AddVectorToPosition(
     Position=Linked() {linkedVariable {name=Particles.ShapeLocation.ShapeNormal, type=Vector3f}}
     Vector=&RandomRangeVector(Minimum=<0 -0.05 0>, Maximum=<0 -0.25 0>)
    )) {isSetParameters=true, module=SetVariables_#3}
   &ParticleState()
   &ScaleSpriteSize(
    Non-Uniform Curve Sprite Scale=DataInterface() {propertyValues {
      XCurve {keys [{tangentMode=RCTM_User, time=0, value=0}, {tangentMode=RCTM_User, time=0.2, value=1}]}
      YCurve {keys [
        <at 0 0 RCIM_Cubic>
        {arriveTangent=1.25, interpMode=RCIM_Cubic, leaveTangent=1.25, time=0.2, value=1}
        {arriveTangent=-1.25, interpMode=RCIM_Cubic, leaveTangent=-1.25, time=0.8, value=1}
        <at 1 0 RCIM_Cubic>
       ]}
     }}
    Scale Sprite Size Mode=::NewEnumerator3
   )
   /SetVariables(Particles.GlowSize=&Multiply_Vector2D(A=$Particles.SpriteSize, B=&Multiply_Vector2D_ByFloat(Float=1, Vector2D=<2.5 2.5>))) {isSetParameters=true, module=SetVariables_#5}
   &DynamicMaterialParameters(
    Index 0 Param 2=&FloatFromCurve(
     FloatCurve=DataInterface() {propertyValues {Curve {keys [{tangentMode=RCTM_User, time=0, value=0}, {tangentMode=RCTM_User, time=1, value=1}]}}}
     Scale Curve=&Lerp_Float(A=0.5, B=2, Alpha=$User.StateLerp)
    )
    Module.Param0WriteEnabled=false
    Module.Param1WriteEnabled=true
    Module.Param2WriteEnabled=true
    Module.Param3WriteEnabled=true
   ) {module=DynamicMaterialParameters001}
  ]
  rendererClasses [&NiagaraSpriteRendererProperties]
  renderers [
   {rendererClass=&NiagaraSpriteRendererProperties}
   {rendererClass=&NiagaraSpriteRendererProperties, rendererIndex=1}
  ]
 }
}
summary {
 systemName=NS_Backpack_Mouse_Idle
 emitters [
  {
   emitterName=Lightning_Small
   rendererClasses [&NiagaraMeshRendererProperties, &NiagaraSpriteRendererProperties]
  }
  {emitterName=Heart, rendererClasses [&NiagaraMeshRendererProperties]}
  {emitterName=Spikes, rendererClasses [&NiagaraSpriteRendererProperties]}
  {emitterName=SpikesMain, rendererClasses [&NiagaraSpriteRendererProperties]}
  {emitterName=SpikesEnergy_Main, rendererClasses [&NiagaraSpriteRendererProperties]}
  {
   emitterName=Lightning_Large
   rendererClasses [&NiagaraMeshRendererProperties, &NiagaraSpriteRendererProperties]
  }
  {emitterName=CoreFlare, rendererClasses [&NiagaraSpriteRendererProperties, &NiagaraLightRendererProperties]}
 ]
 userVariables [
  User.Skeletal Mesh {
   type=NiagaraDataInterfaceSkeletalMesh
   defaultValue {
    struct=NiagaraExt_VariableValue_DataInterface
    value {dataInterface=&self:NiagaraDataInterfaceSkeletalMesh_0}
   }
  }
  User.StateLerp {type=NiagaraFloat, defaultValue=0, description=""}
 ]
}
