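# Animation Blueprint for the F_MED_Mouse backpack cosmetic that gives its dyn_pack bone secondary motion and a floating wobble.
# Each update builds Crown_Float from real-time sine waves at a different frequency per axis.
# A low-alpha spring controller moves dyn_pack, then an additive bone translation applies Crown_Float on top.
# The spring's alpha drops to zero when pawn speed exceeds a high cutoff, disabling it during fast movement.
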AnimBlueprint /BRCosmetics/Accessories/FORT_Backpacks/Backpack_F_MED_Mouse/Meshes/F_MED_Mouse_Pack_AnimBP
target_skeleton=&/F_MED_Mouse_Pack_Skeleton
graphs [
 AnimGraph {nodes [
   n1 = Spring controller - Bone: dyn_pack!(
    ComponentPose:Component Space Pose Link Structure=Local To Component(LocalPose=Input Pose())
    Alpha:float=Get pawnVelocitySpeedCutoff() -> (pawnVelocitySpeedCutoff:double)
   ) -> (Pose:Component Space Pose Link Structure) {is=title AnimGraphNode_SpringBone, alphaBoolBlend {blendOption=Linear}, springBone {boneName=dyn_pack}}
   n2 = Get Crown_Float() -> (Crown_Float:Vector)
   Output Pose(Result=Component To Local(ComponentPose="Transform (Modify) Bone - Bone: dyn_pack"(ComponentPose=@n1, Alpha=1, Translation=@n2, Rotation="0,0,0", Scale="1,1,1") {translationMode=BMM_Additive, alphaBoolBlend {blendOption=Linear}, boneToModify {boneName=dyn_pack}})) {layerGroup=None}
  ]}
 EventGraph {nodes [
   Event BlueprintUpdateAnimation()
   n1 = GetRealTimeSeconds()
   n2 = "Sin (Radians)"(A=float * float(A=float * float(A=1.1, B=1.1), B=@n1))
   n3 = "Sin (Radians)"(A=float * float(A=float * float(A=1.3, B=1.3), B=@n1))
   n4 = "Sin (Radians)"(A=float * float(A=float * float(A=0.9, B=1.1), B=@n1))
   > Set Crown_Float(Crown_Float:Vector=MakeVector(X=float * float(A=@n3, B=0.7), Y=float * float(A=@n2, B=0.7), Z=float * float(A=@n4, B=0.7))) -> (Output_Get:Vector)
   > n5 = Set pawnVelocitySpeed(pawnVelocitySpeed:double=Vector Length(A=Get PawnVelocity())) -> (Output_Get:double)
   > Set pawnVelocitySpeedCutoff(pawnVelocitySpeedCutoff:double=SelectFloat(A=0, B=0.03, bPickA="float > float"(A=@n5, B=1200) {is=title K2Node_CallFunction})) -> (Output_Get:double)
   TryGetPawnOwner()
  ]}
]
variables [
 Crown_Float {type=Vector}
 pawnVelocitySpeed {type=number}
 pawnVelocitySpeedCutoff {type=number}
]
